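# Parent material for the CrackCrispTundra cosmetic shoe skin, adding a toggleable, near-white emissive mask to the shared character material.
# The skin's mask texture R and G channels form the glow, which CCT_BlinkingToggle scales before adding to base emissive.
# A thresholded sine timer drives a lerp intended for blinking, but both inputs are R, so it stays constant.
# Shared reactivity, sock, PostFX and hologram functions then process the attributes.
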Material /CosmeticShoes/Assets/FORT_Shoes/Shoe_StemDotTempo/Skins/CrackCrispTundra/Materials/M_CrackCrispTundra_ShoeParent
graph {
 properties {
  bUseMaterialAttributes=true
  bUsedWithClothing=true
  bUsedWithMeshDeformer=true
  bUsedWithMorphTargets=true
  bUsedWithNanite=true
  bUsedWithNiagaraMeshParticles=true
  bUsedWithSkeletalMesh=true
  pixelDepthOffsetMode=PDOM_Legacy
  shadingModel=MSM_FromMaterialExpression
  usedShadingModels=Default Lit | Subsurface | Clear Coat | Subsurface Profile | Cloth
 }
 nodes [
  n1 = MaterialFunctionCall() {materialFunction=&MF_CharacterParent_2023_PreFX}
  n2 = GetMaterialAttributes(None=@n1.Result) {attributeGetTypes [#1, #2]}
  n3 = MaterialFunctionCall!(Material=SetMaterialAttributes!(MaterialAttributes=@n2.MaterialAttributes, Base Color, Emissive Color=NamedRerouteUsage()) {attributeSetTypes [#1, #2]}) {materialFunction=/Game/Global/Materials/Cosmetics/Kicks/Shoes/Materials/MF_ShoeReactivityTest}
  MaterialFunctionCall!(In=MaterialFunctionCall(
    Material=MaterialFunctionCall!(InOpacity, Material=@n3.Result, Mask=@n1.Tech Art Mask) {materialFunction=/Game/Global/Materials/Cosmetics/Kicks/Socks/Materials/MF_Sock}.Result
    M=@n1.M
    Pre Skinned Local Position=@n1.M
    Distance Blend=@n1.M
    Fresnel=@n1.M
   ) {materialFunction=&MF_CharacterParent_2023_PostFX}.Result) {materialFunction=/Game/Global/Materials/Cosmetics/Kicks/Shoes/Materials/MF_TechartShoeHologram}
  NamedRerouteDeclaration(None=@n2.Emissive Color) {name=Emissive_IN, nodeColor=<rgba 0.8516 1 0 1>}
  n4 = TextureSample() {
   samplerType=SAMPLERTYPE_Masks
   texture=/CosmeticShoes/Assets/FORT_Shoes/Shoe_StemDotTempo/Skins/CrackCrispTundra/Textures/T_CrackCrispTundra_Shoe_TA
  }
  n5 = Constant() {r=0.47}
  n6 = LinearInterpolate(
   A=@n4.R
   B=@n4.R
   Alpha=SmoothStep(Value=MaterialFunctionCall(Speed=@n5) {materialFunction=&DebugTimeSine}.Result) {constMax=0.51, constMin=0.49}
  )
  n7 = Constant() {r=0.2}
  n8 = Multiply(
   A=Add(A=Multiply(A=@n4.G, B=@n6), B=Multiply(A=@n6, B=@n7) {constB=0.05}) {constB=0.3}
   B=Constant3Vector() {constant=<rgba 0.9583 0.9583 0.9583 0>}
  )
  n9 = ScalarParameter() {defaultValue=1, group=CrackCrispTundra, parameterName=CCT_BlinkingToggle}
  NamedRerouteDeclaration(None=Add(A=NamedRerouteUsage(), B=Multiply(A=@n8, B=@n9))) {name=Emissive_OUT, nodeColor=<rgba 0.8281 0 1 1>}
 ]
}
